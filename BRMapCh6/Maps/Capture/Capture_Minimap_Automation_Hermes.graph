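# Editor utility task that automates minimap capture for the Hermes_Terrain map in BRMapCh6.
# Chained async tasks clear the automated textures, set the developer playlist, run Capture_Minimap at 4K with the discovered scene capture, then stop PIE.
# CaptureToFinalTargets selects whether the capture also imports into Apollo_Terrain_Minimap.
# A map-only capture branch is unconnected, and the 8K and undiscovered capture variables are unused.
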EditorUtilityBlueprint /BRMapCh6/Maps/Capture/Capture_Minimap_Automation_Hermes
graphs [EventGraph {nodes [
   Event BeginExecution()
   > n1 = Construct Object from Class!(
    execute
    Class=/Game/Athena/Editor/ClearTextures.ClearTextures_C
    self=Self-Reference()
    TexturesToReset:Array of Texture 2D Object References=Make Array!(
     "[0]":Texture 2D Object Reference
     "[1]":Texture 2D Object Reference=/BRMapCh6/Maps/Capture/Automated/Minimap_4096_Automated
     "[2]":Texture 2D Object Reference=/BRMapCh6/Maps/Capture/Automated/Minimap_8192_Automated
    ) -> (Array:Array of Texture 2D Object References)
   ) -> (then, ReturnValue:Clear Textures Object Reference)
   > n2 = RegisterAndExecuteTask!(execute:Exec, Task:Editor Utility Task Object Reference=@n1, OptionalParentTask:Editor Utility Task Object Reference=Self-Reference()) -> (then:Exec, OnFinished:Exec, Task:Editor Utility Task Object Reference) {is=title K2Node_AsyncAction}
   SetDeveloperSettingPlaylist!(
    execute=@n2.OnFinished
    Playlist:Fort Playlist Athena Soft Object Reference=Get Playlist() -> (Playlist:Fort Playlist Athena Soft Object Reference)
   ) -> (then)
   n3 = Get CaptureToFinalTargets() -> (CaptureToFinalTargets:bool)
   > n4 = Construct Object from Class!(
    execute
    Class=/Game/Athena/Editor/Capture_Minimap.Capture_Minimap_C
    self=Self-Reference()
    LoadEverything:bool=True
    PreserveExistingState:bool=False
    LoadMapOnly:bool=False
    Resolution:int=4096
    Map:Soft Object Path Structure=Get CaptureMap() -> (CaptureMap:Soft Object Path Structure)
    ImportTargets:Array of Texture 2D Object References=Make Array!(
     "[0]"=&Minimap_4096_Automated
     "[1]"=Select(Option 0:Texture 2D Object Reference, Option 1:Texture 2D Object Reference=/Game/Athena/Apollo/Maps/UI/Apollo_Terrain_Minimap, Index=@n3) -> (ReturnValue:Texture 2D Object Reference)
    ) -> (Array)
    Scene Capture Class:Scene Capture 2D Class Reference=Get DiscoveredClass() -> (DiscoveredClass:Scene Capture 2D Class Reference)
   ) -> (then, ReturnValue:Capture Minimap Object Reference)
   > n5 = RegisterAndExecuteTask(Task=@n4, OptionalParentTask=Self-Reference())
   n6 = Construct Object from Class!(execute=@n5.OnFinished, Class=/Game/Athena/Editor/StopPIE.StopPIE_C, self=Self-Reference()) -> (then, ReturnValue:Stop PIE Object Reference)
   > n7 = RegisterAndExecuteTask(Task=@n6, OptionalParentTask=Self-Reference())
   FinishExecutingTask(execute=@n7.OnFinished)
   n8 = Construct Object from Class!(
    execute
    Class=&Capture_Minimap_C
    self=Self-Reference()
    LoadEverything:bool=True
    PreserveExistingState:bool=False
    LoadMapOnly:bool=true
    Resolution:int=0
    Map:Soft Object Path Structure=Get CaptureMap()
    ImportTargets:Array of Texture 2D Object References
    Scene Capture Class:Scene Capture 2D Class Reference
    StreamingDummyActorClass:Actor Class Reference
    Minimap Camera Actor:Camera Actor Object Reference
   ) -> (then, ReturnValue:Capture Minimap Object Reference)
   > RegisterAndExecuteTask(Task=@n8, OptionalParentTask=Self-Reference())
  ]}]
variables [
 Capture8k {type=boolean}
 Capture8k Undiscovered {type=boolean}
 CaptureMap {type=SoftObjectPath, default=/BRMapCh6/Maps/Hermes_Terrain}
 CaptureToFinalTargets {type=boolean}
 DiscoveredClass {type=Class, default=/Game/Athena/Editor/SceneCapture_Discovered.SceneCapture_Discovered_C}
 Minimap Camera Actor {type=CameraActor, default=null}
 Playlist {type=FortPlaylistAthena, default=/BRPlaylists/Athena/Playlists/Playlist_DefaultSolo}
 UndiscoveredClass {type=Class, default=/Game/Athena/Editor/SceneCapture_Undiscovered.SceneCapture_Undiscovered_C}
]
